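; Parent Blueprint for Rufus plant props, a building actor whose construction script lets collision be switched on or off through a variable.
; The construction script runs the parent class's logic, then reads the CollisionEnabled boolean.
; It sets the static mesh component to QueryAndPhysics when true and NoCollision when false.
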(fn ConstructionScript
  (bind _staticmeshcomponent (Variables|Debug|GetStaticMeshComponent))
  (|CallParentFunction)
  (if (Variables|Default|GetCollisionEnabled)
    (Collision|SetCollisionEnabled _staticmeshcomponent "QueryAndPhysics")
    (else
      (Collision|SetCollisionEnabled _staticmeshcomponent "NoCollision"))))
